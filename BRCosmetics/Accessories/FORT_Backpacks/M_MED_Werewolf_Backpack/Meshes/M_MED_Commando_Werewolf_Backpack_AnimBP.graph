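# Animation Blueprint for the Commando Werewolf backpack cosmetic that offsets the pack as the wearer's spine bends.
# Each update reads the owner mesh's spine_05 delta from ref pose relative to spine_01, mapping negative yaw through clamped ranges to a roll and Y offset.
# The AnimGraph adds both to backpack_root in bone space on top of the input pose.
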AnimBlueprint /BRCosmetics/Accessories/FORT_Backpacks/M_MED_Werewolf_Backpack/Meshes/M_MED_Commando_Werewolf_Backpack_AnimBP
target_skeleton=&/M_MED_Commando_Werewolf_Backpack_Skeleton
graphs [
 AnimGraph {nodes [
   AnimGraphNode_SubInput_0 = Input Pose()
   n1 = "Transform (Modify) Bone - Bone: backpack_root"(
    ComponentPose=Local To Component(LocalPose=@AnimGraphNode_SubInput_0)
    Alpha=1
    Translation=Get BackpackTranslation_b() -> (BackpackTranslation_b:Vector)
    Rotation=Get BackpackRotation_b() -> (BackpackRotation_b:Rotator)
    Scale="1,1,1"
   ) {
    rotationMode=BMM_Additive
    rotationSpace=BCS_BoneSpace
    translationMode=BMM_Additive
    translationSpace=BCS_BoneSpace
    alphaBoolBlend {blendOption=Linear}
    boneToModify {boneName=backpack_root}
   }
   Output Pose(Result=Component To Local(ComponentPose=@n1)) {layerGroup=None}
  ]}
 EventGraph {nodes [
   Event BlueprintUpdateAnimation()
   n1 = GetOwnerSkeletalMeshComponent()
   n2 = BreakRotator(InRot=BreakTransform(InTransform=GetDeltaTransformFromRefPose(self=@n1, BoneName=spine_05, BaseName=spine_01)).Rotation)
   > Set BackpackRotation_b(BackpackRotation_b:Rotator=MakeRotator(Roll=MapRangeClamped(Value=@n2.Yaw, InRangeA=0, InRangeB=-30, OutRangeB=20))) -> (Output_Get:Rotator)
   > Set BackpackTranslation_b(BackpackTranslation_b:Vector=MakeVector(Y=MapRangeClamped(Value=@n2.Yaw, InRangeA=0, InRangeB=-30, OutRangeB=-3))) -> (Output_Get:Vector)
  ]}
]
variables [
 BackpackRotation_b {type=Rotator}
 BackpackTranslation_b {type=Vector}
]
